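# Post-process style animation Blueprint for the M_MED_LopexSnow body cosmetic, copying the attached parent's pose and correcting its secondary bones.
# Constraints plus AnimDynamics drive the zipper, while constraints and additive bone offsets shape the collar and thigh extras.
# Update logic derives alphas from head and thigh deltas, aiming, emotes and movement states, feeding those offsets and a final Control Rig.
# The front-end upper-arm correction sits unconnected to the output.
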AnimBlueprint /BRCosmetics/Characters/Player/Male/Medium/Bodies/M_MED_LopexSnow/Meshes/M_MED_LopexSnow_AnimBP
target_skeleton=/Game/Characters/Player/Male/Medium/Bodies/M_MED_LopexSnow/Meshes/M_MED_LopexSnow_Skeleton
graphs [
 AnimGraph {nodes [
   "Transform (Modify) Bone - Bone: upperarm_r"(
    Alpha=Get Upperarm_r_frontend_alpha_1() -> (Upperarm_r_frontend_alpha_1:double)
    Translation="0,0,0"
    Rotation="-8.805,6.505,-12.821"
    Scale="1,1,1"
   ) {rotationMode=BMM_Additive, alphaBoolBlend {blendOption=Linear}, boneToModify {boneName=upperarm_r}, rotation=<rot -8.805 6.505 -12.821>}
   n1 = "Constraint - dyn_zipper (PPPPPPPP)"!(
    ComponentPose=Local To Component(LocalPose=Copy Pose From Mesh() {bUseAttachedParent=true})
    Alpha=1
    ConstraintWeights_0=0.0521
    ConstraintWeights_1:float=0.0494
    ConstraintWeights_2:float=0.2019
    ConstraintWeights_3:float=0.2825
    ConstraintWeights_4:float=0.3299
    ConstraintWeights_5:float=0.0357
    ConstraintWeights_6:float=0.0367
    ConstraintWeights_7:float=0.0117
   ) -> (Pose) {
    alphaBoolBlend {blendOption=Linear}
    boneToModify {boneName=dyn_zipper}
    constraintSetup [
     {targetBone {boneName=clavicle_l}}
     {targetBone {boneName=spine_01}}
     {targetBone {boneName=spine_02}}
     {targetBone {boneName=spine_03}}
     {targetBone {boneName=spine_04}}
     {targetBone {boneName=spine_05}}
     {targetBone {boneName=clavicle_r}}
     {targetBone {boneName=pelvis}}
    ]
    constraintWeights [1,1,1,1,1,1,1,1]
   }
   n2 = Anim Dynamics - Bone: dyn_zipper(ComponentPose=@n1, Alpha=1) {
    angularDampingOverride=1
    angularSpringConstant=1500
    bGravityOverrideInSimSpace=true
    bOverrideAngularDamping=true
    gravityScale=0.35
    simulationSpace=RootRelative
    alphaBoolBlend {blendOption=Linear}
    boundBone {boneName=dyn_zipper}
    physicsBodyDefinitions [{
      sphereCollisionRadius=0
      boundBone {boneName=dyn_zipper}
      boxExtents=<2 5 5>
      constraintSetup {angularLimitsMax=<0 60 0>, angularLimitsMin=<0 -60 -90>, angularTarget=<0 0 0>}
      localJointOffset=<2 0 0>
     }]
   }
   n3 = "Constraint - dyn_collar_back_01 (P)"(ComponentPose=@n2, Alpha=1, ConstraintWeights_0=1) {
    alphaBoolBlend {blendOption=Linear}
    boneToModify {boneName=dyn_collar_back_01}
    constraintSetup [{targetBone {boneName=neck_02}}]
    constraintWeights [1]
   }
   n4 = "Constraint - dyn_collar_front_mid_01_r (P)"(ComponentPose=@n3, Alpha=1, ConstraintWeights_0=1) {
    alphaBoolBlend {blendOption=Linear}
    boneToModify {boneName=dyn_collar_front_mid_01_r}
    constraintSetup [{targetBone {boneName=neck_02}}]
    constraintWeights [1]
   }
   n5 = "Constraint - dyn_collar_mid_01_r (P)"(ComponentPose=@n4, Alpha=1, ConstraintWeights_0=1) {
    alphaBoolBlend {blendOption=Linear}
    boneToModify {boneName=dyn_collar_mid_01_r}
    constraintSetup [{targetBone {boneName=neck_02}}]
    constraintWeights [1]
   }
   n6 = "Constraint - dyn_collar_front_mid_01_l (R)"(ComponentPose=@n5, Alpha=0.2, ConstraintWeights_0=1) {
    alphaBoolBlend {blendOption=Linear}
    boneToModify {boneName=dyn_collar_front_mid_01_l}
    constraintSetup [{transformType=Rotation, targetBone {boneName=clavicle_l}}]
    constraintWeights [1]
   }
   n7 = "Constraint - dyn_collar_mid_01_l (R)"(ComponentPose=@n6, Alpha=0.2, ConstraintWeights_0=1) {
    alphaBoolBlend {blendOption=Linear}
    boneToModify {boneName=dyn_collar_mid_01_l}
    constraintSetup [{transformType=Rotation, targetBone {boneName=clavicle_l}}]
    constraintWeights [1]
   }
   n8 = "Constraint - dyn_collar_mid_01_l (P)"(ComponentPose=@n7, Alpha=1, ConstraintWeights_0=1) {
    alphaBoolBlend {blendOption=Linear}
    boneToModify {boneName=dyn_collar_mid_01_l}
    constraintSetup [{targetBone {boneName=neck_02}}]
    constraintWeights [1]
   }
   n9 = "Transform (Modify) Bone - Bone: thigh_extra_l"(ComponentPose=@n8, Alpha=Get thigh_extra_l_alpha() -> (thigh_extra_l_alpha:double), Translation="0,-20,0", Rotation="0,0,0", Scale="1,1,1") {translationMode=BMM_Additive, alphaBoolBlend {blendOption=Linear}, boneToModify {boneName=thigh_extra_l}, translation {y=-20}}
   n10 = "Transform (Modify) Bone - Bone: thigh_extra_r"(ComponentPose=@n9, Alpha=Get thigh_extra_r_alpha() -> (thigh_extra_r_alpha:double), Translation="0,-20,0", Rotation="0,0,0", Scale="1,1,1") {translationMode=BMM_Additive, alphaBoolBlend {blendOption=Linear}, boneToModify {boneName=thigh_extra_r}, translation {y=-20}}
   n11 = "Transform (Modify) Bone - Bone: dyn_collar_front_mid_01_l"(ComponentPose=@n10, Alpha=Get look_left_alpha() -> (look_left_alpha:double), Translation="0,0,0", Rotation="-20,0,0", Scale="1,1,1") {rotationMode=BMM_Additive, alphaBoolBlend {blendOption=Linear}, boneToModify {boneName=dyn_collar_front_mid_01_l}, rotation {pitch=-20}}
   n12 = "Transform (Modify) Bone - Bone: dyn_collar_front_mid_01_l"(
    ComponentPose=@n11
    Alpha=Get look_down_alpha() -> (look_down_alpha:double)
    Translation="-0.000,-0.000,-1.928"
    Rotation="-17.330,6.718,-2.536"
    Scale="1,1,1"
   ) {
    rotationMode=BMM_Additive
    translationMode=BMM_Additive
    alphaBoolBlend {blendOption=Linear}
    boneToModify {boneName=dyn_collar_front_mid_01_l}
    rotation=<rot -17.33 6.718 -2.536>
    translation {z=-1.928}
   }
   n13 = "Transform (Modify) Bone - Bone: dyn_collar_front_mid_01_l"(
    ComponentPose=@n12
    Alpha=Get tilt_left_alpha() -> (tilt_left_alpha:double)
    Translation="0,0,0"
    Rotation="-18.338,11.966,5.192"
    Scale="1,1,1"
   ) {
    rotationMode=BMM_Additive
    alphaBoolBlend {blendOption=Linear}
    boneToModify {boneName=dyn_collar_front_mid_01_l}
    rotation=<rot -18.338 11.966 5.192>
   }
   n14 = "Transform (Modify) Bone - Bone: dyn_collar_front_01_l"(ComponentPose=@n13, Alpha=Get look_left_alpha(), Translation="0,0,0", Rotation="-10,0,0", Scale="1,1,1") {rotationMode=BMM_Additive, alphaBoolBlend {blendOption=Linear}, boneToModify {boneName=dyn_collar_front_01_l}, rotation {pitch=-10}}
   n15 = "Transform (Modify) Bone - Bone: dyn_collar_mid_01_l"(ComponentPose=@n14, Alpha=Get look_left_alpha(), Translation="0,0,0", Rotation="-75,0,0", Scale="1,1,1") {rotationMode=BMM_Additive, alphaBoolBlend {blendOption=Linear}, boneToModify {boneName=dyn_collar_mid_01_l}, rotation {pitch=-75}}
   n16 = "Transform (Modify) Bone - Bone: dyn_collar_mid_01_l"(ComponentPose=@n15, Alpha=Get look_down_alpha(), Translation="0,0,0", Rotation="-39.280,-46.783,34.565", Scale="1,1,1") {
    rotationMode=BMM_Additive
    alphaBoolBlend {blendOption=Linear}
    boneToModify {boneName=dyn_collar_mid_01_l}
    rotation=<rot -39.28 -46.783 34.565>
   }
   n17 = "Transform (Modify) Bone - Bone: dyn_collar_front_mid_01_r"(ComponentPose=@n16, Alpha=Get look_right_alpha() -> (look_right_alpha:double), Translation="0,0,0", Rotation="40,0,-1.692", Scale="1,1,1") {
    rotationMode=BMM_Additive
    alphaBoolBlend {blendOption=Linear}
    boneToModify {boneName=dyn_collar_front_mid_01_r}
    rotation {pitch=40, roll=-1.692}
   }
   n18 = "Transform (Modify) Bone - Bone: dyn_collar_front_01_r"(
    ComponentPose=@n17
    Alpha=Get left_max_alpha() -> (left_max_alpha:double)
    Translation="0,0,0"
    Rotation="-1.478,-16.232,8.051"
    Scale="1,1,1"
   ) {
    rotationMode=BMM_Additive
    alphaBoolBlend {blendOption=Linear}
    boneToModify {boneName=dyn_collar_front_01_r}
    rotation=<rot -1.478 -16.232 8.051>
   }
   n19 = "Transform (Modify) Bone - Bone: dyn_collar_front_mid_01_l"(
    ComponentPose=@n18
    Alpha=Get head_is_aiming_collar_max_alpha() -> (head_is_aiming_collar_max_alpha:double)
    Translation="-0.000,0.000,-1.425"
    Rotation="2.776,-8.828,35.979"
    Scale="1,1,1"
   ) {
    rotationMode=BMM_Additive
    translationMode=BMM_Additive
    alphaBoolBlend {blendOption=Linear}
    boneToModify {boneName=dyn_collar_front_mid_01_l}
    rotation=<rot 2.776 -8.828 35.979>
    translation {z=-1.425}
   }
   n20 = "Transform (Modify) Bone - Bone: dyn_collar_back_01"(
    ComponentPose=@n19
    Alpha=Get Collar_emote_alpha() -> (Collar_emote_alpha:double)
    Translation="-0.622,-3.238,0.132"
    Rotation="0.000,-32.131,5.890"
    Scale="1,1,1"
   ) {
    rotationMode=BMM_Additive
    rotationSpace=BCS_BoneSpace
    translationMode=BMM_Additive
    translationSpace=BCS_BoneSpace
    alphaBoolBlend {blendOption=Linear}
    boneToModify {boneName=dyn_collar_back_01}
    rotation=<rot 0 -32.131 5.89>
    translation=<-0.622 -3.238 0.132>
   }
   n21 = "Transform (Modify) Bone - Bone: dyn_collar_mid_01_l"(ComponentPose=@n20, Alpha=Get Collar_emote_alpha(), Translation="0.151,-1.991,0.000", Rotation="0.000,-34.936,-0.000", Scale="1,1,1") {
    rotationMode=BMM_Additive
    rotationSpace=BCS_BoneSpace
    translationMode=BMM_Additive
    translationSpace=BCS_BoneSpace
    alphaBoolBlend {blendOption=Linear}
    boneToModify {boneName=dyn_collar_mid_01_l}
    rotation=<rot 0 -34.936 0>
    translation=<0.1511 -1.9909 0>
   }
   n22 = "Transform (Modify) Bone - Bone: dyn_collar_mid_01_r"(ComponentPose=@n21, Alpha=Get Collar_emote_alpha(), Translation="0.135,-2.868,-0.000", Rotation="0.000,-36.316,0.000", Scale="1,1,1") {
    rotationMode=BMM_Additive
    rotationSpace=BCS_BoneSpace
    translationMode=BMM_Additive
    translationSpace=BCS_BoneSpace
    alphaBoolBlend {blendOption=Linear}
    boneToModify {boneName=dyn_collar_mid_01_r}
    rotation=<rot 0 -36.316 0>
    translation=<0.135 -2.868>
   }
   n23 = Control Rig!(
    Source=Component To Local(ComponentPose=@n22)
    Alpha
    skydive_floating_alpha:double=Get skydive_floating_alpha() -> (skydive_floating_alpha:double)
    skydive_floating_alpha_not:double=Get skydive_floating_alpha_not() -> (skydive_floating_alpha_not:double)
    skydive_alpha:double=Get skydive_alpha() -> (skydive_alpha:double)
    skydive_alpha_not:double=Get skydive_alpha_not() -> (skydive_alpha_not:double)
    skydive_dive_alpha:double=Get skydive_dive_alpha() -> (skydive_dive_alpha:double)
    skydive_dive_alpha_not:double=Get skydive_dive_alpha_not() -> (skydive_dive_alpha_not:double)
    skydive_parachute_alpha:double=Get skydive_parachute_alpha() -> (skydive_parachute_alpha:double)
    skydive_parachute_alpha_not:double=Get skydive_parachute_alpha_not() -> (skydive_parachute_alpha_not:double)
    emote_alpha:double=Get emote_alpha() -> (emote_alpha:double)
    emote_alpha_not:double=Get emote_alpha_not() -> (emote_alpha_not:double)
    melee_jump_fall_sprint_alpha_05_1:double=Get melee_jump_fall_sprint_alpha_05_1() -> (melee_jump_fall_sprint_alpha_05_1:double)
    melee_jump_fall_sprint_alpha_0_1:double=Get melee_jump_fall_sprint_alpha_0_1() -> (melee_jump_fall_sprint_alpha_0_1:double)
   ) -> (Pose) {
    alphaBoolBlend {blendOption=Linear}
    alphaScaleBiasClamp {interpSpeedDecreasing=3, interpSpeedIncreasing=3}
    controlRigAssetReference {blueprintRigClass=/Game/Characters/Player/Male/Medium/Bodies/M_MED_LopexSnow/Meshes/M_MED_LopexSnow_CtrlRig_M_MED_RedPepper_CtrlRig.M_MED_LopexSnow_CtrlRig_M_MED_RedPepper_CtrlRig_C}
   }
   Output Pose(Result=@n23) {layerGroup=None}
   Reroute Node(InputPin:Wildcard) -> (OutputPin:Wildcard)
  ]}
 EventGraph {nodes [
   n1 = Event BlueprintUpdateAnimation()
   > n2 = Sequence!(execute) -> (
    then_0
    then_1
    then_2:Exec
    then_3:Exec
    then_4:Exec
    then_5:Exec
    then_6:Exec
    then_7:Exec
    then_8:Exec
    then_9:Exec
    then_10:Exec
    then_11:Exec
    then_12:Exec
    then_13:Exec
   )
   Set DeltaTimeX(execute=@n2.then_10, DeltaTimeX:double=@n1.DeltaTimeX) -> (Output_Get:double)
   n3 = Branch(execute=@n2.then_11, Condition=Get bIsSkyDiving())
   > n4 = Set skydive_alpha(skydive_alpha:double=FInterpTo_Constant!(Current:double=Get skydive_alpha(), Target:double=1, DeltaTime:double=Get DeltaTimeX() -> (DeltaTimeX:double), InterpSpeed:double=2) -> (ReturnValue:double)) -> (Output_Get:double)
   Set skydive_alpha_not(
    execute=[@n5.then, @n4.then]
    skydive_alpha_not:double=MapRangeClamped(Value=Get skydive_alpha(), InRangeA=0, InRangeB=1, OutRangeA=1, OutRangeB=0)
   ) -> (Output_Get:double)
   n5 = Set skydive_alpha(execute=@n3.else, skydive_alpha=FInterpTo_Constant(Current=Get skydive_alpha(), Target=0, DeltaTime=Get DeltaTimeX(), InterpSpeed=2))
   n6 = Branch(execute=@n2.then_12, Condition=Get bIsSkydiveFloating())
   > n7 = Set skydive_floating_alpha(skydive_floating_alpha:double=FInterpTo_Constant(Current=Get skydive_floating_alpha(), Target=1, DeltaTime=Get DeltaTimeX(), InterpSpeed=2)) -> (Output_Get:double)
   Set skydive_floating_alpha_not(
    execute=[@n7.then, @n8.then]
    skydive_floating_alpha_not:double=MapRangeClamped(Value=Get skydive_floating_alpha(), InRangeA=0, InRangeB=1, OutRangeA=1, OutRangeB=0)
   ) -> (Output_Get:double)
   n8 = Set skydive_floating_alpha(
    execute=@n6.else
    skydive_floating_alpha=FInterpTo_Constant(Current=Get skydive_floating_alpha(), Target=0, DeltaTime=Get DeltaTimeX(), InterpSpeed=2)
   )
   n9 = Branch(
    execute=@n2.then_13
    Condition=OR Boolean!(A=Get bIsPlayingMeleeAnim(), B=Get bIsJumping(), C:bool=Get bIsFalling(), D:bool=Get bIsSprinting()) -> (ReturnValue)
   )
   > n10 = Set melee_jump_fall_sprint_alpha_05_1(melee_jump_fall_sprint_alpha_05_1:double=FInterpTo_Constant(Current=Get melee_jump_fall_sprint_alpha_05_1(), Target=0.5, DeltaTime=Get DeltaTimeX(), InterpSpeed=2)) -> (Output_Get:double)
   Set melee_jump_fall_sprint_alpha_0_1(
    execute=[@n11.then, @n10.then]
    melee_jump_fall_sprint_alpha_0_1:double=MapRangeClamped(Value=Get melee_jump_fall_sprint_alpha_05_1(), InRangeA=0.5, InRangeB=1)
   ) -> (Output_Get:double)
   n11 = Set melee_jump_fall_sprint_alpha_05_1(
    execute=@n9.else
    melee_jump_fall_sprint_alpha_05_1=FInterpTo_Constant(Current=Get melee_jump_fall_sprint_alpha_05_1(), Target=1, DeltaTime=Get DeltaTimeX(), InterpSpeed=2)
   )
   n12 = Branch(
    execute=@n2.then_3
    Condition=AND Boolean(A=Get bIsPlayingFullBodySlotInFrontEnd() -> (bIsPlayingFullBodySlotInFrontEnd:bool), B=Get bIsInFrontEnd())
   )
   > Set Upperarm_r_frontend_alpha_1(Upperarm_r_frontend_alpha_1:double=FInterpTo_Constant(Current=Get Upperarm_r_frontend_alpha_1(), Target=1, DeltaTime=Get DeltaTimeX(), InterpSpeed=0.2)) -> (Output_Get:double)
   Set Upperarm_r_frontend_alpha_1(
    execute=@n12.else
    Upperarm_r_frontend_alpha_1=FInterpTo_Constant(Current=Get Upperarm_r_frontend_alpha_1(), Target=0, DeltaTime=Get DeltaTimeX(), InterpSpeed=2)
   )
   n13 = Branch(execute=@n2.then_4, Condition=Get bIsSkydiveDiveMode())
   > n14 = Set skydive_dive_alpha(skydive_dive_alpha:double=FInterpTo_Constant(Current=Get skydive_dive_alpha(), Target=1, DeltaTime=Get DeltaTimeX(), InterpSpeed=2)) -> (Output_Get:double)
   Set skydive_dive_alpha_not(
    execute=[@n14.then, @n15.then]
    skydive_dive_alpha_not:double=MapRangeClamped(Value=Get skydive_dive_alpha(), InRangeA=0, InRangeB=1, OutRangeA=1, OutRangeB=0)
   ) -> (Output_Get:double)
   n15 = Set skydive_dive_alpha(
    execute=@n13.else
    skydive_dive_alpha=FInterpTo_Constant(Current=Get skydive_dive_alpha(), Target=0, DeltaTime=Get DeltaTimeX(), InterpSpeed=2)
   )
   n16 = Branch(execute=@n2.then_5, Condition=Get bIsParachuteOpened())
   > n17 = Set skydive_parachute_alpha(skydive_parachute_alpha:double=FInterpTo_Constant(Current=Get skydive_parachute_alpha(), Target=1, DeltaTime=Get DeltaTimeX(), InterpSpeed=2)) -> (Output_Get:double)
   Set skydive_parachute_alpha_not(
    execute=[@n18.then, @n17.then]
    skydive_parachute_alpha_not:double=MapRangeClamped(Value=Get skydive_parachute_alpha(), InRangeA=0, InRangeB=1, OutRangeA=1, OutRangeB=0)
   ) -> (Output_Get:double)
   n18 = Set skydive_parachute_alpha(
    execute=@n16.else
    skydive_parachute_alpha=FInterpTo_Constant(Current=Get skydive_parachute_alpha(), Target=0, DeltaTime=Get DeltaTimeX(), InterpSpeed=2)
   )
   n19 = Branch(execute=@n2.then_6, Condition=Get bIsTargeting())
   > n20 = Set head_is_aiming_collar_alpha(head_is_aiming_collar_alpha:double=FInterpTo_Constant(Current=Get head_is_aiming_collar_alpha() -> (head_is_aiming_collar_alpha:double), Target=1, DeltaTime=Get DeltaTimeX(), InterpSpeed=20)) -> (Output_Get:double)
   n21 = Get left_max_alpha()
   > Set head_is_aiming_collar_max_alpha(head_is_aiming_collar_max_alpha:double=float * float(A=@n20, B=@n21) {is=title K2Node_PromotableOperator}) -> (Output_Get:double)
   n22 = Set head_is_aiming_collar_alpha(
    execute=@n19.else
    head_is_aiming_collar_alpha=FInterpTo_Constant(Current=Get head_is_aiming_collar_alpha(), Target=0, DeltaTime=Get DeltaTimeX(), InterpSpeed=20)
   )
   n23 = Get left_max_alpha()
   > Set head_is_aiming_collar_max_alpha(head_is_aiming_collar_max_alpha=float * float(A=@n22, B=@n23))
   n24 = Get bIsPlayingEmote()
   n25 = Branch(execute=@n2.then_7, Condition=AND Boolean(A=Get bIsInFrontEnd(), B=NOT Boolean(A=@n24)))
   > n26 = Set Upperarm_r_frontend_alpha_2(Upperarm_r_frontend_alpha_2:double=FInterpTo_Constant(Current=Get Upperarm_r_frontend_alpha_2() -> (Upperarm_r_frontend_alpha_2:double), Target=1, DeltaTime=Get DeltaTimeX(), InterpSpeed=20)) -> (Output_Get:double)
   n27 = GetAlphaFromReferenceTransformDelta(
    SkeletalMeshComponent=Get PartSkelMesh() -> (PartSkelMesh:Skeletal Mesh Component Object Reference)
    BoneName=upperarm_r
    BaseName=lowerarm_r
    InputRange=(X=-50,Y=-30)
    TwistAxis="0,0,1"
   )
   > Set Upperarm_r_frontend_alpha_1(Upperarm_r_frontend_alpha_1=float * float(A=@n26, B=@n27))
   n28 = Set Upperarm_r_frontend_alpha_2(
    execute=@n25.else
    Upperarm_r_frontend_alpha_2=FInterpTo_Constant(Current=Get Upperarm_r_frontend_alpha_2(), Target=0, DeltaTime=Get DeltaTimeX(), InterpSpeed=20)
   )
   > Set Upperarm_r_frontend_alpha_1(Upperarm_r_frontend_alpha_1=float * float(A=@n27, B=@n28))
   n29 = Branch(execute=@n2.then_8, Condition=Get bIsPlayingEmote())
   > Set emote_alpha(emote_alpha:double=FInterpTo_Constant(Current=Get emote_alpha(), Target=1, DeltaTime=Get DeltaTimeX(), InterpSpeed=2.5)) -> (Output_Get:double)
   n30 = GetAlphaFromReferenceTransformDelta(SkeletalMeshComponent=Get PartSkelMesh(), BoneName=head, BaseName=neck_02, InputRange=(X=0,Y=-40), TwistAxis="0,0,1")
   > n31 = Set Collar_emote_alpha(Collar_emote_alpha:double=float * float(A=Get emote_alpha(), B=@n30)) -> (Output_Get:double)
   Set emote_alpha_not(
    execute=[@n31.then, @n32.then]
    emote_alpha_not:double=MapRangeClamped(Value=Get emote_alpha(), InRangeA=0, InRangeB=1, OutRangeA=1, OutRangeB=0)
   ) -> (Output_Get:double)
   Set emote_alpha(execute=@n29.else, emote_alpha=FInterpTo_Constant(Current=Get emote_alpha(), Target=0, DeltaTime=Get DeltaTimeX(), InterpSpeed=0.5))
   > n32 = Set Collar_emote_alpha(Collar_emote_alpha=float * float(A=@n30, B=Get emote_alpha()))
   n33 = GetPartSkeletalMeshComponent()
   Set thigh_extra_l_alpha(
    execute=@n2.then_9
    thigh_extra_l_alpha:double=GetAlphaFromReferenceTransformDelta(
     SkeletalMeshComponent=@n33
     BoneName=pelvis
     BaseName=thigh_l
     DeltaComponentType=QuaternionTwist
     InputRange=(X=0,Y=-100)
     TwistAxis="0,0,1"
    )
   ) -> (Output_Get:double)
   > Set thigh_extra_r_alpha(thigh_extra_r_alpha:double=GetAlphaFromReferenceTransformDelta(
     SkeletalMeshComponent=@n33
     BoneName=pelvis
     BaseName=thigh_r
     DeltaComponentType=QuaternionTwist
     InputRange=(X=0,Y=-100)
     TwistAxis="0,0,1"
    )) -> (Output_Get:double)
   n34 = Get PartSkelMesh()
   n35 = BreakRotator(InRot=BreakTransform(InTransform=GetDeltaTransformFromRefPose(self=@n34, BoneName=head, BaseName=spine_05)).Rotation)
   Set look_left_alpha(execute=@n2.then_1, look_left_alpha:double=MapRangeClamped(Value=@n35, InRangeA=15, InRangeB=40)) -> (Output_Get:double)
   > Set look_right_alpha(look_right_alpha:double=MapRangeClamped(Value=@n35, InRangeA=0, InRangeB=-20)) -> (Output_Get:double)
   n36 = Get PartSkelMesh()
   n37 = BreakRotator(InRot=BreakTransform(InTransform=GetDeltaTransformFromRefPose(self=@n36, BoneName=head, BaseName=spine_05)).Rotation)
   Set look_up_alpha(execute=@n2.then_0, look_up_alpha:double=MapRangeClamped(Value=@n37.Yaw, InRangeA=0, InRangeB=-40)) -> (Output_Get:double)
   > Set look_down_alpha(look_down_alpha:double=MapRangeClamped(Value=@n37.Yaw, InRangeA=-10, InRangeB=25)) -> (Output_Get:double)
   n38 = Get PartSkelMesh()
   n39 = BreakRotator(InRot=BreakTransform(InTransform=GetDeltaTransformFromRefPose(self=@n38, BoneName=head, BaseName=spine_05)).Rotation)
   Set tilt_left_alpha(execute=@n2.then_2, tilt_left_alpha:double=MapRangeClamped(Value=@n39.Pitch, InRangeA=0, InRangeB=-10)) -> (Output_Get:double)
   > Set tilt_right_alpha(tilt_right_alpha:double=MapRangeClamped(Value=@n39.Pitch, InRangeA=0, InRangeB=10)) -> (Output_Get:double)
   n40 = "Max (Float)"(A=Get look_left_alpha(), B=Get tilt_left_alpha())
   > Set left_max_alpha(left_max_alpha:double=@n40) -> (Output_Get:double)
   > Set left_max_down(left_max_down:double="Max (Float)"(A=@n40, B=Get look_down_alpha())) -> (Output_Get:double)
   n41 = "Max (Float)"(A=Get look_right_alpha(), B=Get tilt_right_alpha() -> (tilt_right_alpha:double))
   > Set right_max_alpha(right_max_alpha:double=@n41) -> (Output_Get:double)
   > Set right_max_down(right_max_down:double="Max (Float)"(A=@n41, B=Get look_down_alpha())) -> (Output_Get:double)
   Event BlueprintInitializeAnimation()
   > Set PartSkelMesh(PartSkelMesh:Skeletal Mesh Component Object Reference=@n33) -> (Output_Get:Skeletal Mesh Component Object Reference)
   TryGetPawnOwner()
   MapRangeClamped(Value=@n21, InRangeA=0.2, InRangeB=0.7)
   MapRangeClamped(Value=@n23, InRangeA=0.2, InRangeB=0.7)
   Get look_down_alpha()
   Get look_down_alpha()
   Get left_max_down() -> (left_max_down:double)
   Get left_max_down()
  ]}
]
variables [
 Collar_emote_alpha {type=number}
 DeltaTimeX {type=number}
 emote_alpha {type=number}
 emote_alpha_not {type=number}
 head_is_aiming_collar_alpha {type=number}
 head_is_aiming_collar_max_alpha {type=number}
 left_max_alpha {type=number}
 left_max_alpha_0 {type=number}
 left_max_down {type=number}
 look_down_alpha {type=number}
 look_left_alpha {type=number}
 look_right_alpha {type=number}
 look_up_alpha {type=number}
 melee_jump_fall_sprint_alpha_05_1 {type=number}
 melee_jump_fall_sprint_alpha_0_1 {type=number}
 PartSkelMesh {type=SkeletalMeshComponent}
 right_max_alpha {type=number}
 right_max_down {type=number}
 Shuolderpad_l_alpha {type=number}
 Shuolderpad_r_alpha {type=number}
 skydive_alpha {type=number}
 skydive_alpha_not {type=number}
 skydive_dive_alpha {type=number}
 skydive_dive_alpha_not {type=number}
 skydive_floating_alpha {type=number}
 skydive_floating_alpha_not {type=number}
 skydive_parachute_alpha {type=number}
 skydive_parachute_alpha_not {type=number}
 thigh_extra_l_alpha {type=number}
 thigh_extra_r_alpha {type=number}
 tilt_left_alpha {type=number}
 tilt_right_alpha {type=number}
 Upperarm_r_frontend_alpha_1 {type=number}
 Upperarm_r_frontend_alpha_2 {type=number}
]
